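# Sound cue for the foley of the TNTina emote in Battle Royale cosmetics.
# A mixer layers the main foley wave with a slightly quieter branch that, after a fixed short delay, plays one of three equally weighted distant BlackMonday explosion sounds.
# Playback uses emote foley class, attenuation and concurrency settings, with a submix send to music audio analysis.
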SoundCue /BRCosmetics/Sounds/Emotes/TNTina/Athena_Emote_TNTina_Foley_Cue
root=Mixer_0
nodes [
 {id=Delay_1, children [Random_0], properties {delayMax=0.65, delayMin=0.65}}
 {id=Mixer_0, children [WavePlayer_0,Delay_1], properties {inputVolume [1,0.8]}}
 {id=Random_0, children [WavePlayer_2,WavePlayer_3,WavePlayer_4], properties {weights [1,1,1]}}
 {id=WavePlayer_0, properties {soundWaveAssetPtr=&/Athena_Emote_TNTina_Foley}}
 {id=WavePlayer_2, properties {soundWaveAssetPtr=/Game/Sounds/Fort_Weapons/BlackMonday/Bangs/BlackMonday_Bang_Explode_Distant_01}}
 {id=WavePlayer_3, properties {soundWaveAssetPtr=/Game/Sounds/Fort_Weapons/BlackMonday/Bangs/BlackMonday_Bang_Explode_Distant_02}}
 {id=WavePlayer_4, properties {soundWaveAssetPtr=/Game/Sounds/Fort_Weapons/BlackMonday/Bangs/BlackMonday_Bang_Explode_Distant_03}}
]
properties {
 attenuationSettings=&Fort_Emotes_Athena_Att
 audioPropertiesSheet=/BRCosmetics/Sounds/Settings/PropertySheets/Emotes/APS_Cosmetic_Emote_Foley
 maxDistance=1600
 soundClassObject=/Game/Sounds/FortSoundClassesAndModes/Fort_Foley_Emote
 volumeMultiplier=1
 concurrencyOverrides {maxCount=3, resolutionRule=StopOldest}
 concurrencySet [/Game/Sounds/Concurrency_Rules/Athena_EmoteFoley_Conc]
 soundSubmixSends [{minSendLevel=0, soundSubmix=&Fortnite_AudioAnalysis_Music}]
}
